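# Physics asset for the LadyShinobi female cosmetic body, defining collision and constraints for cloak and belt-tie secondary motion.
# Left and right seven-link cloak chains, a single back segment and two belt ties hang from base bones with tight swing limits and zero twist.
# Capsules on torso, limbs and cape attach, plus a pouch box, form body colliders on mostly angular-free constraints.
# An oversized, downward-offset root capsule resembles a ground proxy.
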PhysicsAsset /BRCosmetics/Characters/Player/Female/Medium/Bodies/F_MED_LadyShinobi/Meshes/F_MED_LadyShinobi_CapePhysics
bodies {
 attach_cape [{shapeType=Capsule, length=75.7516, radius=19.1537, shapeName=pelvis_capsule, center=<0 -7.4234 -35.7876>, rotation=<rot 0 0 -4.6614>}]
 base_BeltTie [{shapeType=Capsule, length=0.505, radius=0.505, shapeName=base_BeltTie_capsule, rotation=<rot 0 -90 0>}]
 base_Cloak_b [{shapeType=Capsule, length=0.505, radius=0.505, shapeName=base_Cloak_b_capsule, rotation=<rot 0 -90 0>}]
 base_Cloak_l [{shapeType=Capsule, length=0.505, radius=0.505, shapeName=base_Cloak_l_capsule, rotation=<rot 0 -90 0>}]
 base_Cloak_r [{shapeType=Capsule, length=0.505, radius=0.505, shapeName=base_Cloak_r_capsule, rotation=<rot 0 -90 0>}]
 calf_twist_02_l [{
   shapeType=Capsule
   length=29.0814
   radius=7.7225
   shapeName=calf_twist_01_l_capsule
   center=<-5.8832 -2.6067 0>
   rotation=<rot 90 180 -5.4675>
  }]
 calf_twist_02_r [{
   shapeType=Capsule
   length=29.0814
   radius=7.7225
   shapeName=calf_twist_01_r_capsule
   center=<5.8832 2.6067 0>
   rotation=<rot 90 179.4032 173.9357>
  }]
 dyn_BeltTie_l [{
   shapeType=Capsule
   length=2.7223
   radius=1.2206
   shapeName=dyn_BeltTie_l_capsule
   center=<2.971 0.1096 0.3742>
   rotation=<rot 0 91.0102 -84.8567>
  }]
 dyn_BeltTie_r [{
   shapeType=Capsule
   length=3.007
   radius=1.3173
   shapeName=dyn_BeltTie_r_capsule
   center=<2.9315 0.1782 0.2641>
   rotation=<rot 0 85.912 -89.8455>
  }]
 dyn_Cloak_b_01 [{
   shapeType=Capsule
   length=11.9616
   radius=3.0706
   shapeName=dyn_Cloak_b_01_capsule
   center=<5.137 0.1041 -0.224>
   rotation=<rot 0 178.3508 -86.3174>
  }]
 dyn_Cloak_l_01 [{
   shapeType=Capsule
   length=11.1767
   radius=2.3243
   shapeName=dyn_Cloak_l_01_capsule
   center=<6.5863 -0.7407 -0.0299>
   rotation=<rot 0.2258 6.4067 -94.6208>
  }]
 dyn_Cloak_l_02 [{
   shapeType=Capsule
   length=6.5204
   radius=3.1732
   shapeName=dyn_Cloak_l_02_capsule
   center=<4.8979 -0.0289 -0.1245>
   rotation=<rot 0 0 -85.7104>
  }]
 dyn_Cloak_l_03 [{shapeType=Capsule, length=4.7315, radius=3.55, shapeName=dyn_Cloak_l_03_capsule, center=<6.012 -0.1435 -0.6031>, rotation=<rot 0 0 -89.7706>}]
 dyn_Cloak_l_04 [{shapeType=Capsule, length=5.3244, radius=3.5517, shapeName=dyn_Cloak_l_04_capsule, center=<7.8021 0 0>, rotation=<rot 0 0 -88.0761>}]
 dyn_Cloak_l_05 [{
   shapeType=Capsule
   length=5.0666
   radius=3.9539
   shapeName=dyn_Cloak_l_05_capsule
   center=<7.0092 0 0>
   rotation=<rot 0.0003 -0.0025 -85.8233>
  }]
 dyn_Cloak_l_06 [{shapeType=Capsule, length=8.2207, radius=3.5945, shapeName=dyn_Cloak_l_06_capsule, center=<6.2441 0 0>, rotation=<rot 0 -0.0001 -83.6992>}]
 dyn_Cloak_l_07 [{shapeType=Capsule, length=10.9744, radius=4.283, shapeName=dyn_Cloak_l_07_capsule, center=<9.1956 0.1149 0.482>, rotation=<rot 0 0 -88.5364>}]
 dyn_Cloak_r_01 [{
   shapeType=Capsule
   length=11.1767
   radius=2.3243
   shapeName=dyn_Cloak_r_01_capsule
   center=<5.5599 0.8386 -0.4926>
   rotation=<rot 3.1389 -6.8535 108.5731>
  }]
 dyn_Cloak_r_02 [{
   shapeType=Capsule
   length=6.5204
   radius=3.1732
   shapeName=dyn_Cloak_r_02_capsule
   center=<4.8979 0.0289 0.1245>
   rotation=<rot 0 -0.0001 103.0686>
  }]
 dyn_Cloak_r_03 [{shapeType=Capsule, length=4.7315, radius=3.55, shapeName=dyn_Cloak_r_03_capsule, center=<7.2189 0.1435 0.6031>, rotation=<rot 0 0 103.386>}]
 dyn_Cloak_r_04 [{
   shapeType=Capsule
   length=5.3244
   radius=3.5517
   shapeName=dyn_Cloak_r_04_capsule
   center=<7.089 0.1448 0.6667>
   rotation=<rot 0.0006 -0.0052 106.1346>
  }]
 dyn_Cloak_r_05 [{
   shapeType=Capsule
   length=5.5201
   radius=3.5005
   shapeName=dyn_Cloak_r_05_capsule
   center=<7.0092 0.158 0.7287>
   rotation=<rot 0.0003 -0.0025 104.8053>
  }]
 dyn_Cloak_r_06 [{shapeType=Capsule, length=8.2207, radius=3.5945, shapeName=dyn_Cloak_r_06_capsule, center=<6.2441 0 0>, rotation=<rot 0 -0.0002 99.705>}]
 dyn_Cloak_r_07 [{
   shapeType=Capsule
   length=10.9744
   radius=4.283
   shapeName=dyn_Cloak_r_07_capsule
   center=<9.1956 -0.1149 -0.482>
   rotation=<rot 0 0 103.4073>
  }]
 foot_l [{shapeType=Capsule, length=10.3151, radius=7.4212, shapeName=foot_l_capsule, center=<-1.969 5.6724 -0.0895>, rotation=<rot 0 6.3951 91.2154>}]
 foot_r [{shapeType=Capsule, length=10.3151, radius=7.4212, shapeName=foot_r_capsule, center=<1.969 -5.6724 0.0895>, rotation=<rot 0 6.3951 -88.7846>}]
 hand_l [{
   shapeType=Capsule
   length=7.7094
   radius=5.7423
   shapeName=hand_l_capsule
   center=<7.2702 -2.3918 -0.5193>
   rotation=<rot 90 -41.229 154.7205>
  }]
 hand_r [{
   shapeType=Capsule
   length=7.7094
   radius=5.7423
   shapeName=hand_r_capsule
   center=<-7.2702 2.3918 0.5193>
   rotation=<rot 90 -41.229 -25.2794>
  }]
 lowerarm_twist_02_l [{shapeType=Capsule, length=16.8733, radius=4.3114, shapeName=lowerarm_twist_01_l_capsule, rotation=<rot 90 0 0>}]
 lowerarm_twist_02_r [{shapeType=Capsule, length=16.8733, radius=4.3114, shapeName=lowerarm_twist_01_r_capsule, rotation=<rot 90 180 0>}]
 pelvis [
  {shapeType=Capsule, length=8.6718, radius=10.1739, shapeName=pelvis_capsule, center=<-3.8342 -2.496 0>}
  {shapeType=Capsule, length=71.7007, radius=21.4565, shapeName=pelvis_capsule, center=<-10.5263 -31.7657 0>, rotation=<rot 90 0 0>}
 ]
 pouch_attach [{
   shapeType=Box
   extentX=16.3111
   extentY=7.5
   extentZ=7.5
   shapeName=thigh_twist_01_r_box
   center=<-0.8505 -2.6917 -0.2403>
   rotation=<rot 0 -0.0001 -11.2226>
  }]
 root [{shapeType=Capsule, length=67.6788, radius=150.0058, shapeName=root_capsule, center=<0 -22.5149 -147.0646>, rotation=<rot 90 0 0>}]
 spine_01 [
  {shapeType=Capsule, length=12.693, radius=7.2004, shapeName=spine_01_capsule, center=<0 -1.6611 0>}
  {shapeType=Capsule, length=13.0996, radius=4.9336, shapeName=spine_01_capsule, center=<-3.3176 4.9603 0>}
 ]
 spine_02 [{shapeType=Capsule, length=1.4164, radius=9.0551, shapeName=spine_02_capsule}]
 spine_03 [{shapeType=Capsule, length=9.9348, radius=3.9347, shapeName=spine_03_capsule}]
 spine_04 [{shapeType=Capsule, length=6.9005, radius=7.1453, shapeName=spine_04_capsule}]
 spine_05 [{shapeType=Capsule, length=2.9762, radius=10.2367, shapeName=spine_05_capsule}]
 thigh_twist_01_l [{shapeType=Capsule, length=34.7655, radius=6.8005, shapeName=thigh_twist_01_l_capsule, center=<3.3197 0 0>, rotation=<rot 90 180 0>}]
 thigh_twist_01_r [{shapeType=Capsule, length=34.7655, radius=6.8005, shapeName=thigh_twist_01_r_capsule, center=<-3.3197 0 0>, rotation=<rot 90 180 180>}]
 upperarm_twist_01_l [{shapeType=Capsule, length=19.7975, radius=4.2905, shapeName=upperarm_twist_01_l_capsule, center=<3.2784 0 0>, rotation=<rot 90 0 180>}]
 upperarm_twist_01_r [{shapeType=Capsule, length=19.7975, radius=4.2905, shapeName=upperarm_twist_01_r_capsule, center=<-3.2784 0 0>, rotation=<rot 90 0 0>}]
}
constraints [
 {bone1Name=dyn_Cloak_l_01, bone2Name=base_Cloak_l, swing1LimitDegrees=0, swing2LimitDegrees=5, twistLimitDegrees=0}
 {bone1Name=dyn_Cloak_l_02, bone2Name=dyn_Cloak_l_01, swing1LimitDegrees=5, swing2LimitDegrees=5, twistLimitDegrees=0}
 {bone1Name=dyn_Cloak_l_03, bone2Name=dyn_Cloak_l_02, swing1LimitDegrees=5, swing2LimitDegrees=5, twistLimitDegrees=0}
 {bone1Name=dyn_Cloak_l_04, bone2Name=dyn_Cloak_l_03, swing1LimitDegrees=5, swing2LimitDegrees=5, twistLimitDegrees=0}
 {bone1Name=dyn_Cloak_l_05, bone2Name=dyn_Cloak_l_04, swing1LimitDegrees=5, swing2LimitDegrees=5, twistLimitDegrees=0}
 {bone1Name=dyn_Cloak_l_06, bone2Name=dyn_Cloak_l_05, swing1LimitDegrees=5, swing2LimitDegrees=5, twistLimitDegrees=0}
 {bone1Name=dyn_Cloak_l_07, bone2Name=dyn_Cloak_l_06, swing1LimitDegrees=5, swing2LimitDegrees=5, twistLimitDegrees=0}
 {bone1Name=dyn_Cloak_r_01, bone2Name=base_Cloak_r, swing1LimitDegrees=0, swing2LimitDegrees=5, twistLimitDegrees=0}
 {bone1Name=dyn_Cloak_r_02, bone2Name=dyn_Cloak_r_01, swing1LimitDegrees=5, swing2LimitDegrees=5, twistLimitDegrees=0}
 {bone1Name=dyn_Cloak_r_03, bone2Name=dyn_Cloak_r_02, swing1LimitDegrees=5, swing2LimitDegrees=5, twistLimitDegrees=0}
 {bone1Name=dyn_Cloak_r_04, bone2Name=dyn_Cloak_r_03, swing1LimitDegrees=5, swing2LimitDegrees=5, twistLimitDegrees=0}
 {bone1Name=dyn_Cloak_r_05, bone2Name=dyn_Cloak_r_04, swing1LimitDegrees=5, swing2LimitDegrees=5, twistLimitDegrees=0}
 {bone1Name=dyn_Cloak_r_06, bone2Name=dyn_Cloak_r_05, swing1LimitDegrees=5, swing2LimitDegrees=5, twistLimitDegrees=0}
 {bone1Name=dyn_Cloak_r_07, bone2Name=dyn_Cloak_r_06, swing1LimitDegrees=5, swing2LimitDegrees=5, twistLimitDegrees=0}
 {bone1Name=dyn_BeltTie_l, bone2Name=base_BeltTie, swing1LimitDegrees=10, swing2LimitDegrees=10, twistLimitDegrees=0}
 {bone1Name=dyn_BeltTie_r, bone2Name=base_BeltTie, swing1LimitDegrees=10, swing2LimitDegrees=10, twistLimitDegrees=0}
 {bone1Name=dyn_Cloak_b_01, bone2Name=base_Cloak_b, swing1LimitDegrees=15, swing2LimitDegrees=25, twistLimitDegrees=0}
 {
  bone1Name=base_Cloak_l
  bone2Name=spine_05
  swing1LimitDegrees=45
  swing1Motion=Free
  swing2LimitDegrees=45
  swing2Motion=Free
  twistLimitDegrees=45
  twistMotion=Free
 }
 {
  bone1Name=base_Cloak_r
  bone2Name=spine_05
  swing1LimitDegrees=45
  swing1Motion=Free
  swing2LimitDegrees=45
  swing2Motion=Free
  twistLimitDegrees=45
  twistMotion=Free
 }
 {bone1Name=base_Cloak_b, bone2Name=spine_05, swing1LimitDegrees=25, swing2LimitDegrees=25, twistLimitDegrees=0}
 {
  bone1Name=spine_05
  bone2Name=spine_04
  swing1LimitDegrees=45
  swing1Motion=Free
  swing2LimitDegrees=45
  swing2Motion=Free
  twistLimitDegrees=45
  twistMotion=Free
 }
 {
  bone1Name=spine_04
  bone2Name=spine_03
  swing1LimitDegrees=45
  swing1Motion=Free
  swing2LimitDegrees=45
  swing2Motion=Free
  twistLimitDegrees=45
  twistMotion=Free
 }
 {
  bone1Name=spine_03
  bone2Name=spine_02
  swing1LimitDegrees=45
  swing1Motion=Free
  swing2LimitDegrees=45
  swing2Motion=Free
  twistLimitDegrees=45
  twistMotion=Free
 }
 {
  bone1Name=spine_02
  bone2Name=spine_01
  swing1LimitDegrees=45
  swing1Motion=Free
  swing2LimitDegrees=45
  swing2Motion=Free
  twistLimitDegrees=45
  twistMotion=Free
 }
 {
  bone1Name=base_BeltTie
  bone2Name=spine_01
  swing1LimitDegrees=45
  swing1Motion=Free
  swing2LimitDegrees=45
  swing2Motion=Free
  twistLimitDegrees=45
  twistMotion=Free
 }
 {
  bone1Name=spine_01
  bone2Name=pelvis
  swing1LimitDegrees=45
  swing1Motion=Free
  swing2LimitDegrees=45
  swing2Motion=Free
  twistLimitDegrees=45
  twistMotion=Free
 }
 {
  bone1Name=upperarm_twist_01_l
  bone2Name=spine_05
  swing1LimitDegrees=45
  swing1Motion=Free
  swing2LimitDegrees=45
  swing2Motion=Free
  twistLimitDegrees=45
  twistMotion=Free
 }
 {
  bone1Name=upperarm_twist_01_r
  bone2Name=spine_05
  swing1LimitDegrees=45
  swing1Motion=Free
  swing2LimitDegrees=45
  swing2Motion=Free
  twistLimitDegrees=45
  twistMotion=Free
 }
 {
  bone1Name=thigh_twist_01_l
  bone2Name=pelvis
  swing1LimitDegrees=45
  swing1Motion=Free
  swing2LimitDegrees=45
  swing2Motion=Free
  twistLimitDegrees=45
  twistMotion=Free
 }
 {
  bone1Name=thigh_twist_01_r
  bone2Name=pelvis
  swing1LimitDegrees=45
  swing1Motion=Free
  swing2LimitDegrees=45
  swing2Motion=Free
  twistLimitDegrees=45
  twistMotion=Free
 }
 {
  bone1Name=calf_twist_02_r
  bone2Name=pelvis
  swing1LimitDegrees=45
  swing1Motion=Free
  swing2LimitDegrees=45
  swing2Motion=Free
  twistLimitDegrees=45
  twistMotion=Free
 }
 {
  bone1Name=calf_twist_02_l
  bone2Name=pelvis
  swing1LimitDegrees=45
  swing1Motion=Free
  swing2LimitDegrees=45
  swing2Motion=Free
  twistLimitDegrees=45
  twistMotion=Free
 }
 {
  bone1Name=lowerarm_twist_02_l
  bone2Name=spine_05
  swing1LimitDegrees=45
  swing1Motion=Free
  swing2LimitDegrees=45
  swing2Motion=Free
  twistLimitDegrees=45
  twistMotion=Free
 }
 {
  bone1Name=lowerarm_twist_02_r
  bone2Name=spine_05
  swing1LimitDegrees=45
  swing1Motion=Free
  swing2LimitDegrees=45
  swing2Motion=Free
  twistLimitDegrees=45
  twistMotion=Free
 }
 {
  bone1Name=foot_r
  bone2Name=pelvis
  swing1LimitDegrees=45
  swing1Motion=Free
  swing2LimitDegrees=45
  swing2Motion=Free
  twistLimitDegrees=45
  twistMotion=Free
 }
 {
  bone1Name=foot_l
  bone2Name=pelvis
  swing1LimitDegrees=45
  swing1Motion=Free
  swing2LimitDegrees=45
  swing2Motion=Free
  twistLimitDegrees=45
  twistMotion=Free
 }
 {
  bone1Name=pouch_attach
  bone2Name=pelvis
  swing1LimitDegrees=45
  swing1Motion=Free
  swing2LimitDegrees=45
  swing2Motion=Free
  twistLimitDegrees=45
  twistMotion=Free
 }
 {
  bone1Name=pelvis
  bone2Name=root
  swing1LimitDegrees=45
  swing1Motion=Free
  swing2LimitDegrees=45
  swing2Motion=Free
  twistLimitDegrees=45
  twistMotion=Free
 }
 {
  bone1Name=attach_cape
  bone2Name=spine_05
  swing1LimitDegrees=45
  swing1Motion=Free
  swing2LimitDegrees=45
  swing2Motion=Free
  twistLimitDegrees=45
  twistMotion=Free
 }
 {
  bone1Name=hand_l
  bone2Name=spine_05
  swing1LimitDegrees=45
  swing1Motion=Free
  swing2LimitDegrees=45
  swing2Motion=Free
  twistLimitDegrees=45
  twistMotion=Free
 }
 {
  bone1Name=hand_r
  bone2Name=spine_05
  swing1LimitDegrees=45
  swing1Motion=Free
  swing2LimitDegrees=45
  swing2Motion=Free
  twistLimitDegrees=45
  twistMotion=Free
 }
]
