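# Post-process-style AnimBlueprint for the LadyShinobi cosmetic body, adding secondary motion to the pose copied from the parent mesh and Head part.
# Control Rig and bone offsets feed AnimDynamics leg wires, then cached-pose branches: three RigidBody setups (main, cape, front-end) and a rotated-cloak pose, selected by bool blends.
# Events update gravity and sim-space settings by movement state via a deprecated helper, detect backbling/cape, and track the celebration emote.
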AnimBlueprint /BRCosmetics/Characters/Player/Female/Medium/Bodies/F_MED_LadyShinobi/Meshes/F_MED_LadyShinobi_AnimBP
target_skeleton=&/F_MED_LadyShinobi_Skeleton
graphs [
 AnimGraph {nodes [
   n1 = RigidBody!(
    ComponentPose=Local To Component(LocalPose=Use cached pose 'AnimDyn2RBS'())
    Alpha
    OverrideWorldGravity:Vector="0,0,0"
    ExternalForce
    SimSpaceSettings:Sim Space Settings Structure=(WorldAlpha=1,VelocityScaleZ=1,DampingAlpha=1,MaxLinearVelocity=500,MaxAngularVelocity=3,MaxLinearAcceleration=1000,MaxAngularAcceleration=1,ExternalLinearDragV=(X=1,Y=1,Z=1),ExternalLinearVelocity=(X=0,Y=0,Z=0),ExternalAngularVelocity=(X=0,Y=0,Z=0))
   ) -> (Pose) {
    bOverrideWorldGravity=true
    overridePhysicsAsset=&/F_MED_LadyShinobi_Physics
    simulationSpace=BaseBoneSpace
    alphaBoolBlend {blendOption=Linear}
    baseBoneRef {boneName=pelvis}
    componentAppliedLinearAccClamp=<3000 3000 300>
    simSpaceSettings {maxAngularAcceleration=1, maxAngularVelocity=3, maxLinearAcceleration=1000, maxLinearVelocity=500, worldAlpha=1, externalLinearDragV=<1 1 1>}
   }
   n2 = "Transform (Modify) Bone - Bone: base_Cloak_l"(ComponentPose=Local To Component(LocalPose=Use cached pose 'Setup2AnimDyn'()), Alpha=1, Translation="0,0,0", Rotation="5,4,0", Scale="1,1,1") {
    rotationMode=BMM_Additive
    rotationSpace=BCS_BoneSpace
    alphaBoolBlend {blendOption=Linear}
    boneToModify {boneName=base_Cloak_l}
    rotation {pitch=5, yaw=4}
   }
   n3 = "Transform (Modify) Bone - Bone: base_Cloak_r"(ComponentPose=@n2, Alpha=1, Translation="0,0,0", Rotation="5,-4,0", Scale="1,1,1") {
    rotationMode=BMM_Additive
    rotationSpace=BCS_BoneSpace
    alphaBoolBlend {blendOption=Linear}
    boneToModify {boneName=base_Cloak_r}
    rotation {pitch=5, yaw=-4}
   }
   n4 = Blend Poses by bool(BlendPose_0=Component To Local(ComponentPose=@n1), BlendPose_1=Component To Local(ComponentPose=@n3), BlendTime_0=0.8, BlendTime_1=0.8) {blendType=Linear}
   n5 = RigidBody(
    ComponentPose=Local To Component(LocalPose=Use cached pose 'AnimDyn2RBS'())
    OverrideWorldGravity="0,0,0"
    SimSpaceSettings=(WorldAlpha=1,VelocityScaleZ=1,DampingAlpha=1,MaxLinearVelocity=500,MaxAngularVelocity=3,MaxLinearAcceleration=1000,MaxAngularAcceleration=1,ExternalLinearDragV=(X=1,Y=1,Z=1),ExternalLinearVelocity=(X=0,Y=0,Z=0),ExternalAngularVelocity=(X=0,Y=0,Z=0))
   ) {
    bOverrideWorldGravity=true
    overridePhysicsAsset=&/F_MED_LadyShinobi_CapePhysics
    simulationSpace=BaseBoneSpace
    alphaBoolBlend {blendOption=Linear}
    baseBoneRef {boneName=pelvis}
    componentAppliedLinearAccClamp=<3000 3000 300>
    simSpaceSettings {maxAngularAcceleration=1, maxAngularVelocity=3, maxLinearAcceleration=1000, maxLinearVelocity=500, worldAlpha=1, externalLinearDragV=<1 1 1>}
   }
   n6 = Component To Local(ComponentPose=@n5)
   n7 = RigidBody(
    ComponentPose=Local To Component(LocalPose=Use cached pose 'AnimDyn2RBS'())
    OverrideWorldGravity="0,0,0"
    SimSpaceSettings=(WorldAlpha=1,VelocityScaleZ=1,DampingAlpha=1,MaxLinearVelocity=500,MaxAngularVelocity=3,MaxLinearAcceleration=1000,MaxAngularAcceleration=1,ExternalLinearDragV=(X=1,Y=1,Z=1),ExternalLinearVelocity=(X=0,Y=0,Z=0),ExternalAngularVelocity=(X=0,Y=0,Z=0))
   ) {
    bOverrideWorldGravity=true
    overridePhysicsAsset=&/F_MED_LadyShinobi_FE_Physics
    simulationSpace=BaseBoneSpace
    alphaBoolBlend {blendOption=Linear}
    baseBoneRef {boneName=pelvis}
    componentAppliedLinearAccClamp=<3000 3000 300>
    simSpaceSettings {maxAngularAcceleration=1, maxAngularVelocity=3, maxLinearAcceleration=1000, maxLinearVelocity=500, worldAlpha=1, externalLinearDragV=<1 1 1>}
   }
   Output Pose(Result=Blend Poses by bool(BlendPose_0=Component To Local(ComponentPose=@n7), BlendPose_1=Blend Poses by bool(BlendPose_0=@n6, BlendPose_1=@n4) {blendType=Linear}) {blendType=Linear}) {layerGroup=None}
   n8 = "Get Face Mesh "() -> ("Face Mesh ":Skeletal Mesh Component Object Reference)
   Save cached pose 'Start_Cache'(Pose=Layered blend per bone(
     BasePose=Copy Pose From Mesh() {bUseAttachedParent=true}
     BlendPoses_0=Copy Pose From Mesh(SourceMeshComponent=@n8) {bUseAttachedParent=true}
    )) {cacheName=Start_Cache}
   n9 = Anim Dynamics - Bone: dyn_LegWire_l!(
    ComponentPose=Local To Component(LocalPose=Use cached pose 'Setup2AnimDyn'())
    Alpha=1
    GravityOverride:Vector="0,0,-980"
    ExternalForce="0,0,0"
   ) -> (Pose) {
    angularDampingOverride=1
    angularSpringConstant=20000
    bAngularSpring=true
    bOverrideAngularDamping=true
    bUseGravityOverride=true
    gravityScale=0.35
    simulationSpace=RootRelative
    alphaBoolBlend {blendOption=Linear}
    boundBone {boneName=dyn_LegWire_l}
    componentLinearAccScale=<1 1 1>
    componentLinearVelScale=<1 1 1>
    gravityOverride {z=-980}
    physicsBodyDefinitions [{
      boundBone {boneName=dyn_LegWire_l}
      boxExtents=<20 30 30>
      constraintSetup {angularLimitsMax=<15 15 25>, angularLimitsMin=<-15 -15 -25>, angularTarget=<1 0 0>}
      localJointOffset=<4 0 0>
     }]
   }
   n10 = Anim Dynamics - Bone: dyn_LegWire_r!(ComponentPose=@n9, Alpha=1, GravityOverride:Vector="0,0,-980", ExternalForce="0,0,0") -> (Pose) {
    angularDampingOverride=1
    angularSpringConstant=20000
    bAngularSpring=true
    bOverrideAngularDamping=true
    bUseGravityOverride=true
    gravityScale=0.35
    simulationSpace=RootRelative
    alphaBoolBlend {blendOption=Linear}
    boundBone {boneName=dyn_LegWire_r}
    componentLinearAccScale=<1 1 1>
    componentLinearVelScale=<1 1 1>
    gravityOverride {z=-980}
    physicsBodyDefinitions [{
      boundBone {boneName=dyn_LegWire_r}
      boxExtents=<20 30 30>
      constraintSetup {angularLimitsMax=<15 15 25>, angularLimitsMin=<-15 -15 -25>, angularTarget=<1 0 0>}
      localJointOffset=<4 0 0>
     }]
   }
   Save cached pose 'AnimDyn2RBS'(Pose=Component To Local(ComponentPose=@n10)) {cacheName=AnimDyn2RBS}
   n11 = Control Rig(Source=Use cached pose 'Start_Cache'()) {
    alphaBoolBlend {blendOption=Linear}
    controlRigAssetReference {blueprintRigClass=&/F_MED_LadyShinobi_CtrlRig.F_MED_LadyShinobi_CtrlRig_C}
   }
   n12 = "Transform (Modify) Bone - Bone: FACIAL_R_UnderChin"!(ComponentPose=Local To Component(LocalPose=@n11), bAlphaBoolEnabled:bool=True, Translation="0,1,3", Rotation="0,0,0", Scale="1,1,1") -> (Pose) {
    alphaInputType=Bool
    translationMode=BMM_Additive
    translationSpace=BCS_BoneSpace
    alphaBoolBlend {blendOption=Linear}
    boneToModify {boneName=FACIAL_R_UnderChin}
    translation {y=1, z=3}
   }
   n13 = "Transform (Modify) Bone - Bone: FACIAL_C_UnderChin"!(ComponentPose=@n12, bAlphaBoolEnabled:bool=True, Translation="0,-1,3", Rotation="0,0,0", Scale="1,1,1") -> (Pose) {
    alphaInputType=Bool
    translationMode=BMM_Additive
    translationSpace=BCS_BoneSpace
    alphaBoolBlend {blendOption=Linear}
    boneToModify {boneName=FACIAL_C_UnderChin}
    translation {y=-1, z=3}
   }
   n14 = "Transform (Modify) Bone - Bone: FACIAL_L_UnderChin"!(ComponentPose=@n13, bAlphaBoolEnabled:bool=True, Translation="0,1,3", Rotation="0,0,0", Scale="1,1,1") -> (Pose) {
    alphaInputType=Bool
    translationMode=BMM_Additive
    translationSpace=BCS_BoneSpace
    alphaBoolBlend {blendOption=Linear}
    boneToModify {boneName=FACIAL_L_UnderChin}
    translation {y=1, z=3}
   }
   n15 = "Transform (Modify) Bone - Bone: base_Cloak_l"!(ComponentPose=@n14, bAlphaBoolEnabled:bool=True, Translation="0,1,1", Rotation="0,0,0", Scale="1,1,1") -> (Pose) {
    alphaInputType=Bool
    rotationMode=BMM_Additive
    rotationSpace=BCS_BoneSpace
    translationMode=BMM_Additive
    translationSpace=BCS_BoneSpace
    alphaBoolBlend {blendOption=Linear}
    boneToModify {boneName=base_Cloak_l}
    translation {y=1, z=1}
   }
   n16 = "Transform (Modify) Bone - Bone: base_Cloak_r"(ComponentPose=@n15, Alpha=1, Translation="0,0,1.5", Rotation="0,0,0", Scale="1,1,1") {
    rotationMode=BMM_Additive
    rotationSpace=BCS_BoneSpace
    translationMode=BMM_Additive
    translationSpace=BCS_BoneSpace
    alphaBoolBlend {blendOption=Linear}
    boneToModify {boneName=base_Cloak_r}
    translation {z=1.5}
   }
   n17 = "Transform (Modify) Bone - Bone: pouch_attach"(ComponentPose=@n16, Alpha=1, Translation="0,0,0", Rotation="2,0,0", Scale="1,1,1") {
    rotationMode=BMM_Additive
    rotationSpace=BCS_BoneSpace
    translationMode=BMM_Additive
    translationSpace=BCS_BoneSpace
    alphaBoolBlend {blendOption=Linear}
    boneToModify {boneName=pouch_attach}
    rotation {pitch=2}
   }
   Save cached pose 'Setup2AnimDyn'(Pose=Component To Local(ComponentPose=@n17)) {cacheName=Setup2AnimDyn}
  ]}
 EventGraph {nodes [
   Event BlueprintUpdateAnimation()
   > Sequence()
   > n1 = Gravity Override State Function DEPRECATED(
    CharacterPartAnimInstance=Self-Reference()
    GameTimeInSeconds=GetGameTimeInSeconds()
    GravityOverrideParams=Get Gravity Override Params()
    RBAN Params=Get RBAN Params()
   )
   > Set Gravity Override(Gravity Override=@n1)
   > Set Sim Space Settings(Sim Space Settings:Sim Space Settings Structure=Break RBANSetting Struct!(RBANSettingStruct:"RBANSetting Struct Structure (by ref)"=@n1.RBAN SettingsOut) -> (
     ComponentLinearAccScale_28_#1:Vector=~
     ComponentLinearVelScale_31_#2:Vector=~
     SimSpaceSettings_34_#3:Sim Space Settings Structure
    ) {is=title K2Node_BreakStruct}.SimSpaceSettings_34_#3) -> (Output_Get:Sim Space Settings Structure)
   n2 = GetAnimInstance(self=GetOwnerSkeletalMeshComponent())
   > Set IsPlayCelebrateEmote(IsPlayCelebrateEmote:double="To Float (Boolean)"(InBool=Montage_IsPlaying(self=@n2, Montage=/BRCosmetics/Animation/Game/MainPlayer/Emotes/Emote_Celebration_Female_M))) -> (Output_Get:double)
   Event BlueprintInitializeAnimation()
   > Set Skel Mesh(Skel Mesh=GetOwnerSkeletalMeshComponent())
   Event OnCharacterCustomizationLoaded()
   > n3 = Sequence()
   > n4 = Cast To PlayerPawn_Athena(Object=TryGetPawnOwner())
   > "Set Face Mesh "("Face Mesh ":Skeletal Mesh Component Object Reference=GetSkeletalMeshForPartType(self=@n4, PartType=Head)) -> (Output_Get:Skeletal Mesh Component Object Reference)
   n5 = CheckForBackbling()
   Set Is Wearing BP(execute=@n3.then_1, Is Wearing BP:bool=OR Boolean!(A=@n5, B=@n5.bIsBackblingACape, C:bool=@n5.bIsBackblingLong) -> (ReturnValue)) -> (Output_Get:bool)
   > Set Is Wearing Cape(Is Wearing Cape=@n5.bIsBackblingLong)
  ]}
]
variables [
 "Face Mesh " {type=SkeletalMeshComponent}
 Gravity Override {type=Vector}
 Gravity Override Params {type=GravityOverrideParamsStruct}
 Is Wearing BP {type=boolean}
 Is Wearing Cape {type=boolean}
 IsPlayCelebrateEmote {type=number}
 RBAN Params {
  type=RBANParamsStruct
  default {
   dBNO {simSpaceSettings {
     maxAngularAcceleration=1
     maxAngularVelocity=3
     maxLinearAcceleration=1000
     maxLinearVelocity=500
     velocityScaleZ=1
     externalLinearDragV=<1 1 1>
    }}
   dBNO_Carried {simSpaceSettings {
     maxAngularAcceleration=1
     maxAngularVelocity=3
     maxLinearAcceleration=1000
     maxLinearVelocity=500
     velocityScaleZ=1
     externalLinearDragV=<1 1 1>
    }}
   emoteOrMelee {simSpaceSettings {
     maxAngularAcceleration=1
     maxAngularVelocity=3
     maxLinearAcceleration=1000
     maxLinearVelocity=500
     velocityScaleZ=1
     externalLinearDragV=<1 1 1>
    }}
   front_End {simSpaceSettings {
     maxAngularAcceleration=1
     maxAngularVelocity=3
     maxLinearAcceleration=1000
     maxLinearVelocity=500
     velocityScaleZ=0.5
     worldAlpha=0.5
     externalLinearDragV=<1 1 1>
    }}
   onGround_Crouch_Moving {simSpaceSettings {
     maxAngularAcceleration=1
     maxAngularVelocity=3
     maxLinearAcceleration=1000
     maxLinearVelocity=500
     velocityScaleZ=1
     externalLinearDragV=<1 1 1>
    }}
   onGround_Falling {simSpaceSettings {
     maxAngularAcceleration=1
     maxAngularVelocity=3
     maxLinearAcceleration=1000
     maxLinearVelocity=500
     velocityScaleZ=1
     externalLinearDragV=<1 1 1>
    }}
   onGround_Moving {simSpaceSettings {
     maxAngularAcceleration=1
     maxAngularVelocity=3
     maxLinearAcceleration=1000
     maxLinearVelocity=500
     velocityScaleZ=1
     externalLinearDragV=<1 1 1>
    }}
   onGround_Standing {simSpaceSettings {
     maxAngularAcceleration=1
     maxAngularVelocity=3
     maxLinearAcceleration=1000
     maxLinearVelocity=500
     velocityScaleZ=1
     externalLinearDragV=<1 1 1>
    }}
   rail_Grinding {simSpaceSettings {
     maxAngularAcceleration=1
     maxAngularVelocity=3
     maxLinearAcceleration=1000
     maxLinearVelocity=500
     velocityScaleZ=1
     worldAlpha=1
     externalLinearDragV=<1 1 1>
    }}
   riding_Moving {simSpaceSettings {
     maxAngularAcceleration=1
     maxAngularVelocity=3
     maxLinearAcceleration=1000
     maxLinearVelocity=500
     velocityScaleZ=1
     externalLinearDragV=<1 1 1>
    }}
   riding_Moving_Targeting {simSpaceSettings {
     maxAngularAcceleration=1
     maxAngularVelocity=3
     maxLinearAcceleration=1000
     maxLinearVelocity=500
     velocityScaleZ=1
     externalLinearDragV=<1 1 1>
    }}
   riding_Standing {simSpaceSettings {
     maxAngularAcceleration=1
     maxAngularVelocity=3
     maxLinearAcceleration=1000
     maxLinearVelocity=500
     velocityScaleZ=1
     externalLinearDragV=<1 1 1>
    }}
   skydive_DivingDown {simSpaceSettings {
     maxAngularAcceleration=1
     maxAngularVelocity=3
     maxLinearAcceleration=1000
     maxLinearVelocity=500
     velocityScaleZ=1
     externalLinearDragV=<1 1 1>
    }}
   skydive_FloatingDown {
    componentLinearAccScale=<0 0.5 0>
    componentLinearVelScale=<0 0.5 0>
    simSpaceSettings {
     maxAngularAcceleration=1
     maxAngularVelocity=3
     maxLinearAcceleration=1000
     maxLinearVelocity=500
     velocityScaleZ=1
     externalLinearDragV=<1 1 1>
    }
   }
   skydive_GoingUp {simSpaceSettings {
     maxAngularAcceleration=1
     maxAngularVelocity=3
     maxLinearAcceleration=1000
     maxLinearVelocity=500
     velocityScaleZ=1
     externalLinearDragV=<1 1 1>
    }}
   skydive_Parachuting {
    componentLinearAccScale=<0 0.1 0.3>
    componentLinearVelScale=<0 0.1 0.3>
    simSpaceSettings {
     maxAngularAcceleration=1
     maxAngularVelocity=3
     maxLinearAcceleration=1000
     maxLinearVelocity=500
     velocityScaleZ=1
     externalLinearDragV=<1 1 1>
    }
   }
   skydive_Surfing {simSpaceSettings {
     maxAngularAcceleration=1
     maxAngularVelocity=3
     maxLinearAcceleration=1000
     maxLinearVelocity=500
     velocityScaleZ=1
     externalLinearDragV=<1 1 1>
    }}
   swimming {simSpaceSettings {
     maxAngularAcceleration=1
     maxAngularVelocity=3
     maxLinearAcceleration=1000
     maxLinearVelocity=500
     velocityScaleZ=1
     externalLinearDragV=<1 1 1>
    }}
   tactical_Sprint {simSpaceSettings {
     maxAngularAcceleration=1
     maxAngularVelocity=3
     maxLinearAcceleration=1000
     maxLinearVelocity=500
     velocityScaleZ=1
     externalLinearDragV=<1 1 1>
    }}
   ziplining {simSpaceSettings {
     maxAngularAcceleration=1
     maxAngularVelocity=3
     maxLinearAcceleration=1000
     maxLinearVelocity=500
     velocityScaleZ=1
     externalLinearDragV=<1 1 1>
    }}
  }
 }
 Sim Space Settings {
  type=SimSpaceSettings
  default {
   dampingAlpha=1
   maxAngularAcceleration=10000
   maxAngularVelocity=10000
   maxLinearAcceleration=10000
   maxLinearVelocity=10000
   velocityScaleZ=1
   worldAlpha=0
   externalLinearDragV=<0 0 0>
  }
 }
 Skel Mesh {type=SkeletalMeshComponent}
]
