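# Physics asset for the LadyShinobi cosmetic body, providing capsule colliders for cape simulation.
# Capsules cover pelvis, spine, clavicles, head and feet, while arm, leg, neck and leg-wire bodies have no shapes.
# Two seven-bone side cloak chains and one back cloak bone carry capsules and attach to spine_05 with free angular motion.
# Chain links have tight angular limits, and each chain root also constrains to a base_Cloak bone lacking a body here.
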PhysicsAsset /BRCosmetics/Characters/Player/Female/Medium/Bodies/F_MED_LadyShinobi/Meshes/F_MED_LadyShinobi_CapeColliders
bodies {
 calf_l []
 calf_r []
 clavicle_l [{shapeType=Capsule, length=6.5623, radius=7.5194, center=<3.9566 -1.0862 -2.1127>, rotation=<rot -77.5566 177.3915 7.7314>}]
 clavicle_r [{shapeType=Capsule, length=6.5623, radius=7.5194, center=<-3.9566 1.0862 2.1127>, rotation=<rot -77.5566 177.3915 -172.2686>}]
 dyn_Cloak_b_01 [{
   shapeType=Capsule
   length=16.1319
   radius=8.6183
   shapeName=dyn_Cloak_b_01_capsule
   center=<3.8122 0.1422 -0.224>
   rotation=<rot 0 178.3508 -86.3174>
  }]
 dyn_Cloak_l_01 [{
   shapeType=Capsule
   length=7.6941
   radius=5.8069
   shapeName=dyn_Cloak_l_01_capsule
   center=<6.5863 -0.7407 -0.0299>
   rotation=<rot 0.2258 6.4067 -94.6208>
  }]
 dyn_Cloak_l_02 [{
   shapeType=Capsule
   length=3.0378
   radius=6.6557
   shapeName=dyn_Cloak_l_02_capsule
   center=<4.8979 -0.0289 -0.1245>
   rotation=<rot 0 0 -85.7104>
  }]
 dyn_Cloak_l_03 [{
   shapeType=Capsule
   length=1.2489
   radius=7.0326
   shapeName=dyn_Cloak_l_03_capsule
   center=<6.012 -0.1435 -0.6031>
   rotation=<rot 0 0 -89.7706>
  }]
 dyn_Cloak_l_04 [{shapeType=Capsule, length=1.8419, radius=7.0342, shapeName=dyn_Cloak_l_04_capsule, center=<7.8021 0 0>, rotation=<rot 0 0 -88.0761>}]
 dyn_Cloak_l_05 [{
   shapeType=Capsule
   length=1.5841
   radius=7.4365
   shapeName=dyn_Cloak_l_05_capsule
   center=<7.0092 0 0>
   rotation=<rot 0.0003 -0.0025 -85.8233>
  }]
 dyn_Cloak_l_06 [{shapeType=Capsule, length=4.7382, radius=7.0771, shapeName=dyn_Cloak_l_06_capsule, center=<6.2441 0 0>, rotation=<rot 0 -0.0001 -83.6992>}]
 dyn_Cloak_l_07 [{shapeType=Capsule, length=7.4919, radius=7.7656, shapeName=dyn_Cloak_l_07_capsule, center=<9.1956 0.1149 0.482>, rotation=<rot 0 0 -88.5364>}]
 dyn_Cloak_r_01 [{
   shapeType=Capsule
   length=7.6941
   radius=5.8069
   shapeName=dyn_Cloak_r_01_capsule
   center=<5.5599 0.8386 -0.4926>
   rotation=<rot 3.1389 -6.8535 108.5731>
  }]
 dyn_Cloak_r_02 [{
   shapeType=Capsule
   length=3.0378
   radius=6.6557
   shapeName=dyn_Cloak_r_02_capsule
   center=<4.8979 0.0289 0.1245>
   rotation=<rot 0 -0.0001 103.0686>
  }]
 dyn_Cloak_r_03 [{shapeType=Capsule, length=1.2489, radius=7.0326, shapeName=dyn_Cloak_r_03_capsule, center=<7.2189 0.1435 0.6031>, rotation=<rot 0 0 103.386>}]
 dyn_Cloak_r_04 [{
   shapeType=Capsule
   length=1.8419
   radius=7.0342
   shapeName=dyn_Cloak_r_04_capsule
   center=<7.089 0.1448 0.6667>
   rotation=<rot 0.0006 -0.0053 106.1346>
  }]
 dyn_Cloak_r_05 [{
   shapeType=Capsule
   length=2.0375
   radius=6.9831
   shapeName=dyn_Cloak_r_05_capsule
   center=<7.0092 0.158 0.7287>
   rotation=<rot 0.0003 -0.0025 104.8053>
  }]
 dyn_Cloak_r_06 [{shapeType=Capsule, length=4.7382, radius=7.0771, shapeName=dyn_Cloak_r_06_capsule, center=<6.2441 0 0>, rotation=<rot 0 -0.0002 99.705>}]
 dyn_Cloak_r_07 [{
   shapeType=Capsule
   length=7.4919
   radius=7.7656
   shapeName=dyn_Cloak_r_07_capsule
   center=<9.1956 -0.1149 -0.482>
   rotation=<rot 0 0 103.4073>
  }]
 dyn_LegWire_l []
 dyn_LegWire_r []
 foot_l [{shapeType=Capsule, length=14.4364, radius=8.0344, center=<-4.1811 6.0027 -0.338>, rotation=<rot 0.0388 1.4701 -89.3982>}]
 foot_r [{shapeType=Capsule, length=14.4364, radius=8.0344, center=<4.1811 -6.0027 0.338>, rotation=<rot 0.0388 1.4701 90.6018>}]
 head [{shapeType=Capsule, length=0, radius=18, center=<9.0697 0.0746 0>, rotation=<rot 0 -90 0>}]
 lowerarm_l []
 lowerarm_r []
 neck_02 []
 pelvis [{shapeType=Capsule, length=9.6042, radius=11.1441, center=<-2.4471 -2.5289 -0.1981>}]
 spine_01 [{shapeType=Capsule, length=4.3888, radius=14.3638, center=<-8.9021 -0.8973 0>}]
 spine_02 [
  {shapeType=Capsule, length=2.5219, radius=9.7317, center=<-2.6885 -1.2641 0>}
  {shapeType=Capsule, length=20.6875, radius=9.3125, center=<-1.8317 1.9362 0>, rotation=<rot 90 0 0>}
 ]
 spine_03 [
  {shapeType=Capsule, length=6.1863, radius=8.4854, center=<0.0574 0.4416 0>}
  {shapeType=Capsule, length=20.7188, radius=9.2812, center=<-1.2163 0.549 0>, rotation=<rot 90 180 -175.0732>}
 ]
 spine_04 [
  {shapeType=Capsule, length=9.0076, radius=10.3916, center=<5.6234 -0.6779 0>}
  {shapeType=Capsule, length=10.0312, radius=9.2812, center=<-1.2326 1.308 0>, rotation=<rot 90 0 0>}
 ]
 spine_05 [{shapeType=Capsule, length=9.678, radius=12.3571, center=<0.6969 1.91 0>}]
 thigh_l []
 thigh_r []
 upperarm_l []
 upperarm_r []
}
constraints [
 {
  bone1Name=spine_01
  bone2Name=pelvis
  swing1LimitDegrees=45
  swing1Motion=Free
  swing2LimitDegrees=45
  swing2Motion=Free
  twistLimitDegrees=45
  twistMotion=Free
 }
 {
  bone1Name=spine_02
  bone2Name=spine_01
  swing1LimitDegrees=45
  swing1Motion=Free
  swing2LimitDegrees=45
  swing2Motion=Free
  twistLimitDegrees=45
  twistMotion=Free
 }
 {
  bone1Name=spine_03
  bone2Name=spine_02
  swing1LimitDegrees=45
  swing1Motion=Free
  swing2LimitDegrees=45
  swing2Motion=Free
  twistLimitDegrees=45
  twistMotion=Free
 }
 {
  bone1Name=spine_04
  bone2Name=spine_03
  swing1LimitDegrees=45
  swing1Motion=Free
  swing2LimitDegrees=45
  swing2Motion=Free
  twistLimitDegrees=45
  twistMotion=Free
 }
 {
  bone1Name=spine_05
  bone2Name=spine_04
  swing1LimitDegrees=45
  swing1Motion=Free
  swing2LimitDegrees=45
  swing2Motion=Free
  twistLimitDegrees=45
  twistMotion=Free
 }
 {
  bone1Name=clavicle_l
  bone2Name=spine_05
  swing1LimitDegrees=45
  swing1Motion=Free
  swing2LimitDegrees=45
  swing2Motion=Free
  twistLimitDegrees=45
  twistMotion=Free
 }
 {
  bone1Name=upperarm_l
  bone2Name=clavicle_l
  swing1LimitDegrees=45
  swing1Motion=Free
  swing2LimitDegrees=45
  swing2Motion=Free
  twistLimitDegrees=45
  twistMotion=Free
 }
 {
  bone1Name=lowerarm_l
  bone2Name=upperarm_l
  swing1LimitDegrees=45
  swing1Motion=Free
  swing2LimitDegrees=45
  swing2Motion=Free
  twistLimitDegrees=45
  twistMotion=Free
 }
 {
  bone1Name=neck_02
  bone2Name=spine_05
  swing1LimitDegrees=45
  swing1Motion=Free
  swing2LimitDegrees=45
  swing2Motion=Free
  twistLimitDegrees=45
  twistMotion=Free
 }
 {
  bone1Name=head
  bone2Name=neck_02
  swing1LimitDegrees=45
  swing1Motion=Free
  swing2LimitDegrees=45
  swing2Motion=Free
  twistLimitDegrees=45
  twistMotion=Free
 }
 {
  bone1Name=thigh_l
  bone2Name=pelvis
  swing1LimitDegrees=45
  swing1Motion=Free
  swing2LimitDegrees=45
  swing2Motion=Free
  twistLimitDegrees=45
  twistMotion=Free
 }
 {
  bone1Name=foot_l
  bone2Name=calf_l
  swing1LimitDegrees=45
  swing1Motion=Free
  swing2LimitDegrees=45
  swing2Motion=Free
  twistLimitDegrees=45
  twistMotion=Free
 }
 {
  bone1Name=lowerarm_r
  bone2Name=upperarm_r
  swing1LimitDegrees=45
  swing1Motion=Free
  swing2LimitDegrees=45
  swing2Motion=Free
  twistLimitDegrees=45
  twistMotion=Free
 }
 {
  bone1Name=upperarm_r
  bone2Name=clavicle_r
  swing1LimitDegrees=45
  swing1Motion=Free
  swing2LimitDegrees=45
  swing2Motion=Free
  twistLimitDegrees=45
  twistMotion=Free
 }
 {
  bone1Name=clavicle_r
  bone2Name=spine_05
  swing1LimitDegrees=45
  swing1Motion=Free
  swing2LimitDegrees=45
  swing2Motion=Free
  twistLimitDegrees=45
  twistMotion=Free
 }
 {
  bone1Name=thigh_r
  bone2Name=pelvis
  swing1LimitDegrees=45
  swing1Motion=Free
  swing2LimitDegrees=45
  swing2Motion=Free
  twistLimitDegrees=45
  twistMotion=Free
 }
 {
  bone1Name=foot_r
  bone2Name=calf_r
  swing1LimitDegrees=45
  swing1Motion=Free
  swing2LimitDegrees=45
  swing2Motion=Free
  twistLimitDegrees=45
  twistMotion=Free
 }
 {
  bone1Name=calf_r
  bone2Name=thigh_r
  swing1LimitDegrees=45
  swing1Motion=Free
  swing2LimitDegrees=45
  swing2Motion=Free
  twistLimitDegrees=45
  twistMotion=Free
 }
 {
  bone1Name=calf_l
  bone2Name=thigh_l
  swing1LimitDegrees=45
  swing1Motion=Free
  swing2LimitDegrees=45
  swing2Motion=Free
  twistLimitDegrees=45
  twistMotion=Free
 }
 {
  bone1Name=dyn_LegWire_r
  bone2Name=calf_r
  swing1LimitDegrees=45
  swing1Motion=Free
  swing2LimitDegrees=45
  swing2Motion=Free
  twistLimitDegrees=45
  twistMotion=Free
 }
 {
  bone1Name=dyn_LegWire_l
  bone2Name=calf_l
  swing1LimitDegrees=45
  swing1Motion=Free
  swing2LimitDegrees=45
  swing2Motion=Free
  twistLimitDegrees=45
  twistMotion=Free
 }
 {
  bone1Name=dyn_Cloak_b_01
  bone2Name=spine_05
  swing1LimitDegrees=45
  swing1Motion=Free
  swing2LimitDegrees=45
  swing2Motion=Free
  twistLimitDegrees=45
  twistMotion=Free
 }
 {
  bone1Name=dyn_Cloak_l_01
  bone2Name=spine_05
  swing1LimitDegrees=45
  swing1Motion=Free
  swing2LimitDegrees=45
  swing2Motion=Free
  twistLimitDegrees=45
  twistMotion=Free
 }
 {bone1Name=dyn_Cloak_l_02, bone2Name=dyn_Cloak_l_01, swing1LimitDegrees=5, swing2LimitDegrees=5, twistLimitDegrees=0}
 {bone1Name=dyn_Cloak_l_03, bone2Name=dyn_Cloak_l_02, swing1LimitDegrees=5, swing2LimitDegrees=5, twistLimitDegrees=0}
 {bone1Name=dyn_Cloak_l_04, bone2Name=dyn_Cloak_l_03, swing1LimitDegrees=5, swing2LimitDegrees=5, twistLimitDegrees=0}
 {bone1Name=dyn_Cloak_l_05, bone2Name=dyn_Cloak_l_04, swing1LimitDegrees=5, swing2LimitDegrees=5, twistLimitDegrees=0}
 {bone1Name=dyn_Cloak_l_06, bone2Name=dyn_Cloak_l_05, swing1LimitDegrees=5, swing2LimitDegrees=5, twistLimitDegrees=0}
 {bone1Name=dyn_Cloak_l_07, bone2Name=dyn_Cloak_l_06, swing1LimitDegrees=5, swing2LimitDegrees=5, twistLimitDegrees=0}
 {
  bone1Name=dyn_Cloak_r_01
  bone2Name=spine_05
  swing1LimitDegrees=45
  swing1Motion=Free
  swing2LimitDegrees=45
  swing2Motion=Free
  twistLimitDegrees=45
  twistMotion=Free
 }
 {bone1Name=dyn_Cloak_r_02, bone2Name=dyn_Cloak_r_01, swing1LimitDegrees=5, swing2LimitDegrees=5, twistLimitDegrees=0}
 {bone1Name=dyn_Cloak_r_03, bone2Name=dyn_Cloak_r_02, swing1LimitDegrees=5, swing2LimitDegrees=5, twistLimitDegrees=0}
 {bone1Name=dyn_Cloak_r_04, bone2Name=dyn_Cloak_r_03, swing1LimitDegrees=5, swing2LimitDegrees=5, twistLimitDegrees=0}
 {bone1Name=dyn_Cloak_r_05, bone2Name=dyn_Cloak_r_04, swing1LimitDegrees=5, swing2LimitDegrees=5, twistLimitDegrees=0}
 {bone1Name=dyn_Cloak_r_06, bone2Name=dyn_Cloak_r_05, swing1LimitDegrees=5, swing2LimitDegrees=5, twistLimitDegrees=0}
 {bone1Name=dyn_Cloak_r_07, bone2Name=dyn_Cloak_r_06, swing1LimitDegrees=5, swing2LimitDegrees=5, twistLimitDegrees=0}
 {bone1Name=dyn_Cloak_r_01, bone2Name=base_Cloak_r, swing1LimitDegrees=5, swing2LimitDegrees=5, twistLimitDegrees=0}
 {bone1Name=dyn_Cloak_l_01, bone2Name=base_Cloak_l, swing1LimitDegrees=5, swing2LimitDegrees=5, twistLimitDegrees=0}
]
